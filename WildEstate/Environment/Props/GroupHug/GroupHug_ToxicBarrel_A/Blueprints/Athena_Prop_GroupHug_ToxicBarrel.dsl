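; Child Blueprint of a reactive prop for the GroupHug toxic barrel, which catches fire at low health and can explode when destroyed.
; Overrides server damage to skip the parent's random-delay timer: melee kills play normal death effects unless all damage may explode, and other kills explode.
; Below a health threshold it applies an on-fire gameplay effect once, credited to the last instigator when valid.
; ShouldDie always returns true, bypassing burn-down, and the Curie radius fire override is empty.
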(event EventOnDamageServer
  (params (Damage :type Float%20%28single-precision%29) (DamageTags :type Gameplay%20Tag%20Container%20Structure%20%28by%20ref%29) (Momentum :type Vector :default "0, 0, 0") (HitInfo :type Hit%20Result%20Structure%20%28by%20ref%29) (InstigatedBy :type Controller%20Object%20Reference) (DamageCauser :type Actor%20Object%20Reference) (EffectContext :type Gameplay%20Effect%20Context%20Handle%20Structure))
  ;[0.1 0.1 0.75] Override with duplicate functionality to bypass the timer with a random delay
  (bind _asfort_player_controller (Utilities|Casting|CastToFortPlayerController InstigatedBy)
    (:then
      (Variables|Default|SetLastDamageInstigator _asfort_player_controller))
    (:CastFailed
      _))
  (Utilities|FlowControl|Sequence
    (:then_0
      (if (Variables|Bools|GetHealthIsZero)
        (Utilities|FlowControl|DoOnce false
          (:Completed
            ;;[0 0 0] Check to see if any type of damage on death should cause explosion
            (if (not (ScalableFloat|GetValueasBool (Variables|Default|GetHF_AllowAllDamageToExplode)))
              ;;[0 0 0] Check If destroyed by melee damage and do explosion on non melee damage
              ;; default actor behavior
              (bind _returnvalue (MissionHelpers|HasTags DamageTags (Variables|GameplayTags|GetTC_MeleeDamage) true false))
              (if _returnvalue
                ;;[0.1 0.1 0.75] Do normal prop death fx
                (CallFunction|DeathFX)
                (Variables|Default|SetDestroyedByPickAxe true)
                (Variables|Default|SetbShouldDie true)
                (else
                  _)))
            ;;[0.8407 0 1] Explode
            (Variables|Default|SetCanGenerateResources false)
            (CallFunction|ExplodeandDie)))))
    (:then_1
      ;;[0 0 0] Set On Fire
      (if (<= (Health|GetHealthPercent) (Variables|Default|GetOnFireThreshold))
        (Utilities|FlowControl|DoOnce false
          (:Completed
            (Utilities|IsValid (Variables|Default|GetLastDamageInstigator)
              (:Is%20Valid
                (GameplayEffects|FortApplyGameplayEffectToTarget :self (Ability|GetFortAbilitySystemComponent (Pawn|GetControlledPawn (Variables|Default|GetLastDamageInstigator))) :GameplayEffectClass "/Game/Athena/Items/EnvironmentalItems/ReactiveProps/GE_Athena_ReactiveProps_OnFire.GE_Athena_ReactiveProps_OnFire_C" :TargetOfEffect (Ability|GetFortAbilitySystemComponent self) :OptionalSourceObject (Pawn|GetControlledPawn (Variables|Default|GetLastDamageInstigator)) :Level 0.0))
              (:Is%20Not%20Valid
                (Environment|ApplyGlobalEnvironmentGameplayEffecttoActor self "/Game/Athena/Items/EnvironmentalItems/ReactiveProps/GE_Athena_ReactiveProps_OnFire.GE_Athena_ReactiveProps_OnFire_C")))
            (CallFunction|OnFireFx)))))))

(fn ExecuteFireApplicationInRadiusIfCurieEnabled
  (params (Fire%20Application%20Target :type Actor%20Object%20Reference)))

(fn ShouldDie
  (params (Damage :type Float%20%28single-precision%29) (EventInstigator :type Controller%20Object%20Reference) (DamageCauser :type Actor%20Object%20Reference) (bIsTest :type Boolean) (EffectContext :type Gameplay%20Effect%20Context%20Handle%20Structure%20%28by%20ref%29))
  ;[1 1 1] Always returning true since we are not using the burn down functionality and we do not want this actor to ever live beyond normal prop death
  (|CallParentFunction Damage EventInstigator DamageCauser bIsTest EffectContext)
  (return true))

(fn ConstructionScript
  (|CallParentFunction))
