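# Decal material for the GroupHug toxic barrel impact, drawing a liquid splat that erodes with decal lifetime.
# Summed texture masks on offset UVs (partly sphere-mapped, with derivative calculations) pass through a sphere mask into Erode_Alpha for opacity.
# Blended, flattened normals offset screen-space depth lookups and feed a custom refract node, projecting WaterCaustics onto the surface behind.
# Shading path switches select separate normal and mask samples for forward and mobile.
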Material /WildEstate/Environment/Props/GroupHug/GroupHug_ToxicBarrel_A/FX/MD_WildEstate_GroupHug_Decal_Impact
graph {
 expressions {
  MaterialExpressionAdd_0 {
   class=MaterialExpressionAdd
   inputs [
    A {from=MaterialExpressionShadingPathSwitch_1}
    B {from=MaterialExpressionShadingPathSwitch_2}
   ]
  }
  MaterialExpressionAdd_1 {
   class=MaterialExpressionAdd
   inputs [
    A {from=MaterialExpressionMultiply_10}
    B {from=MaterialExpressionConstant2Vector_0}
   ]
  }
  MaterialExpressionAdd_10 {
   class=MaterialExpressionAdd
   inputs [
    A {from=MaterialExpressionMultiply_17}
    B {from=MaterialExpressionConstant2Vector_5}
   ]
  }
  MaterialExpressionAdd_11 {
   class=MaterialExpressionAdd
   inputs [
    A {from=MaterialExpressionReroute_1}
    B {from=MaterialExpressionAdd_0}
   ]
  }
  MaterialExpressionAdd_12 {
   class=MaterialExpressionAdd
   inputs [
    A {from=MaterialExpressionMaterialFunctionCall_5, output=Result}
    B {from=MaterialExpressionMultiply_27}
   ]
  }
  MaterialExpressionAdd_13 {
   class=MaterialExpressionAdd
   inputs [
    A {from=MaterialExpressionAdd_7}
    B {from=MaterialExpressionAdd_14}
   ]
  }
  MaterialExpressionAdd_14 {
   class=MaterialExpressionAdd
   inputs [
    A {from=MaterialExpressionShadingPathSwitch_5}
    B {from=MaterialExpressionShadingPathSwitch_6}
   ]
  }
  MaterialExpressionAdd_15 {
   class=MaterialExpressionAdd
   inputs [
    A {from=MaterialExpressionMultiply_29}
    B {from=MaterialExpressionScalarParameter_1}
   ]
   properties {constB=0.3215}
  }
  MaterialExpressionAdd_16 {
   class=MaterialExpressionAdd
   inputs [
    A {from=MaterialExpressionMultiply_18}
    B {from=MaterialExpressionConstant2Vector_7}
   ]
  }
  MaterialExpressionAdd_17 {
   class=MaterialExpressionAdd
   inputs [
    A {from=MaterialExpressionMultiply_21}
    B {from=MaterialExpressionConstant2Vector_7}
   ]
  }
  MaterialExpressionAdd_18 {
   class=MaterialExpressionAdd
   inputs [
    A {from=MaterialExpressionMultiply_24}
    B {from=MaterialExpressionConstant2Vector_9}
   ]
  }
  MaterialExpressionAdd_19 {
   class=MaterialExpressionAdd
   inputs [
    A {from=MaterialExpressionMultiply_25}
    B {from=MaterialExpressionConstant2Vector_9}
   ]
  }
  MaterialExpressionAdd_2 {
   class=MaterialExpressionAdd
   inputs [
    A {from=MaterialExpressionMultiply_2}
    B {from=MaterialExpressionConstant2Vector_0}
   ]
  }
  MaterialExpressionAdd_20 {
   class=MaterialExpressionAdd
   inputs [
    A {from=MaterialExpressionMultiply_26}
    B {from=MaterialExpressionConstant2Vector_11}
   ]
  }
  MaterialExpressionAdd_21 {
   class=MaterialExpressionAdd
   inputs [
    A {from=MaterialExpressionMultiply_33}
    B {from=MaterialExpressionConstant2Vector_11}
   ]
  }
  MaterialExpressionAdd_29 {
   class=MaterialExpressionAdd
   inputs [
    A {from=MaterialExpressionMaterialFunctionCall_86, output=View Space}
    B {from=MaterialExpressionScreenPosition_13, output=ViewportUV}
   ]
  }
  MaterialExpressionAdd_3 {
   class=MaterialExpressionAdd
   inputs [
    A {from=MaterialExpressionMultiply_13}
    B {from=MaterialExpressionConstant2Vector_4}
   ]
  }
  MaterialExpressionAdd_34 {
   class=MaterialExpressionAdd
   inputs [
    A {from=MaterialExpressionMultiply_32}
    B {from=MaterialExpressionMultiply_0}
   ]
  }
  MaterialExpressionAdd_4 {
   class=MaterialExpressionAdd
   inputs [
    A {from=MaterialExpressionMultiply_14}
    B {from=MaterialExpressionConstant2Vector_4}
   ]
  }
  MaterialExpressionAdd_5 {
   class=MaterialExpressionAdd
   inputs [
    A {from=MaterialExpressionShadingPathSwitch_7}
    B {from=MaterialExpressionShadingPathSwitch_8}
   ]
  }
  MaterialExpressionAdd_7 {
   class=MaterialExpressionAdd
   inputs [
    A {from=MaterialExpressionShadingPathSwitch_3}
    B {from=MaterialExpressionShadingPathSwitch_4}
   ]
  }
  MaterialExpressionAdd_8 {
   class=MaterialExpressionAdd
   inputs [
    A {from=MaterialExpressionMultiply_15}
    B {from=MaterialExpressionConstant2Vector_5}
   ]
  }
  MaterialExpressionCameraVectorWS_0 {class=MaterialExpressionCameraVectorWS, inputs [], properties {}}
  MaterialExpressionComponentMask_0 {class=MaterialExpressionComponentMask, inputs [None {from=MaterialExpressionShadingPathSwitch_0}], properties {b=true}}
  MaterialExpressionComponentMask_1 {class=MaterialExpressionComponentMask, inputs [None {from=MaterialExpressionVectorParameter_6, output=RGB}], properties {g=true, r=true}}
  MaterialExpressionComponentMask_10 {class=MaterialExpressionComponentMask, inputs [None {from=MaterialExpressionAdd_4}], properties {g=true}}
  MaterialExpressionComponentMask_11 {class=MaterialExpressionComponentMask, inputs [None {from=MaterialExpressionAdd_10}], properties {r=true}}
  MaterialExpressionComponentMask_12 {class=MaterialExpressionComponentMask, inputs [None {from=MaterialExpressionVectorParameter_11, output=RGB}], properties {g=true, r=true}}
  MaterialExpressionComponentMask_13 {class=MaterialExpressionComponentMask, inputs [None {from=MaterialExpressionAdd_10}], properties {g=true}}
  MaterialExpressionComponentMask_14 {class=MaterialExpressionComponentMask, inputs [None {from=MaterialExpressionAdd_17}], properties {r=true}}
  MaterialExpressionComponentMask_15 {class=MaterialExpressionComponentMask, inputs [None {from=MaterialExpressionAdd_17}], properties {g=true}}
  MaterialExpressionComponentMask_16 {class=MaterialExpressionComponentMask, inputs [None {from=MaterialExpressionMultiply_8}], properties {r=true}}
  MaterialExpressionComponentMask_17 {class=MaterialExpressionComponentMask, inputs [None {from=MaterialExpressionMultiply_8}], properties {g=true}}
  MaterialExpressionComponentMask_18 {class=MaterialExpressionComponentMask, inputs [None {from=MaterialExpressionVertexInterpolator_0, output=PS}], properties {r=true}}
  MaterialExpressionComponentMask_19 {class=MaterialExpressionComponentMask, inputs [None {from=MaterialExpressionVertexInterpolator_0, output=PS}], properties {g=true}}
  MaterialExpressionComponentMask_2 {class=MaterialExpressionComponentMask, inputs [None {from=MaterialExpressionVectorParameter_7, output=RGB}], properties {g=true, r=true}}
  MaterialExpressionComponentMask_20 {class=MaterialExpressionComponentMask, inputs [None {from=MaterialExpressionVertexInterpolator_1, output=PS}], properties {r=true}}
  MaterialExpressionComponentMask_21 {class=MaterialExpressionComponentMask, inputs [None {from=MaterialExpressionVertexInterpolator_1, output=PS}], properties {g=true}}
  MaterialExpressionComponentMask_22 {class=MaterialExpressionComponentMask, inputs [None {from=MaterialExpressionVertexInterpolator_2, output=PS}], properties {r=true}}
  MaterialExpressionComponentMask_23 {class=MaterialExpressionComponentMask, inputs [None {from=MaterialExpressionVertexInterpolator_2, output=PS}], properties {g=true}}
  MaterialExpressionComponentMask_24 {class=MaterialExpressionComponentMask, inputs [None {from=MaterialExpressionAdd_19}], properties {r=true}}
  MaterialExpressionComponentMask_25 {class=MaterialExpressionComponentMask, inputs [None {from=MaterialExpressionAdd_19}], properties {g=true}}
  MaterialExpressionComponentMask_26 {class=MaterialExpressionComponentMask, inputs [None {from=MaterialExpressionAdd_21}], properties {r=true}}
  MaterialExpressionComponentMask_27 {class=MaterialExpressionComponentMask, inputs [None {from=MaterialExpressionAdd_21}], properties {g=true}}
  MaterialExpressionComponentMask_3 {class=MaterialExpressionComponentMask, inputs [None {from=MaterialExpressionVectorParameter_8, output=RGB}], properties {g=true, r=true}}
  MaterialExpressionComponentMask_4 {class=MaterialExpressionComponentMask, inputs [None {from=MaterialExpressionAdd_1}], properties {r=true}}
  MaterialExpressionComponentMask_43 {class=MaterialExpressionComponentMask, inputs [None {from=MaterialExpressionTextureSample_18, output=RGB}], properties {g=true, r=true}}
  MaterialExpressionComponentMask_5 {class=MaterialExpressionComponentMask, inputs [None {from=MaterialExpressionAdd_1}], properties {g=true}}
  MaterialExpressionComponentMask_6 {class=MaterialExpressionComponentMask, inputs [None {from=MaterialExpressionMultiply_12}], properties {r=true}}
  MaterialExpressionComponentMask_7 {class=MaterialExpressionComponentMask, inputs [None {from=MaterialExpressionVectorParameter_9, output=RGB}], properties {g=true, r=true}}
  MaterialExpressionComponentMask_8 {class=MaterialExpressionComponentMask, inputs [None {from=MaterialExpressionMultiply_12}], properties {g=true}}
  MaterialExpressionComponentMask_9 {class=MaterialExpressionComponentMask, inputs [None {from=MaterialExpressionAdd_4}], properties {r=true}}
  MaterialExpressionConstant2Vector_0 {class=MaterialExpressionConstant2Vector, properties {r=0.5}}
  MaterialExpressionConstant2Vector_11 {class=MaterialExpressionConstant2Vector}
  MaterialExpressionConstant2Vector_4 {class=MaterialExpressionConstant2Vector, properties {r=0.5}}
  MaterialExpressionConstant2Vector_5 {class=MaterialExpressionConstant2Vector, properties {g=-0.5}}
  MaterialExpressionConstant2Vector_7 {class=MaterialExpressionConstant2Vector, properties {g=-0.5, r=0.5}}
  MaterialExpressionConstant2Vector_9 {class=MaterialExpressionConstant2Vector, properties {g=-0.5}}
  MaterialExpressionConstant_0 {class=MaterialExpressionConstant, properties {r=0.5}}
  MaterialExpressionConstant_1 {class=MaterialExpressionConstant, properties {r=10}}
  MaterialExpressionConstant_2 {class=MaterialExpressionConstant, properties {r=0.5}}
  MaterialExpressionConstant_3 {class=MaterialExpressionConstant, properties {r=0.5}}
  MaterialExpressionConvert_0 {
   class=MaterialExpressionConvert
   inputs [
    None {from=MaterialExpressionTextureSample_14, output=R}
    None {from=MaterialExpressionMultiply_7}
    None {from=MaterialExpressionTextureSample_14, output=R}
   ]
   properties {
    nodeName=Make Vector3
    convertInputs [
     {
      type=Scalar
      expressionInput {
       expression=&self:MaterialExpressionTextureSample_14
       inputName=None
       mask=1
       maskA=0
       maskB=0
       maskG=0
       maskR=1
       outputIndex=1
      }
     }
     {
      type=Scalar
      expressionInput {
       expression=&self:MaterialExpressionMultiply_7
       inputName=None
       mask=0
       maskA=0
       maskB=0
       maskG=0
       maskR=0
       outputIndex=0
      }
     }
     {
      type=Scalar
      expressionInput {
       expression=&self:MaterialExpressionTextureSample_14
       inputName=None
       mask=1
       maskA=0
       maskB=1
       maskG=0
       maskR=0
       outputIndex=3
      }
     }
    ]
    convertOutputs [{type=Vector3}]
   }
  }
  MaterialExpressionConvert_1 {
   class=MaterialExpressionConvert
   inputs [
    None {from=MaterialExpressionTextureSample_10, output=R}
    None {from=MaterialExpressionMultiply_34}
    None {from=MaterialExpressionTextureSample_10, output=R}
   ]
   properties {
    nodeName=Make Vector3
    convertInputs [
     {
      type=Scalar
      expressionInput {
       expression=&self:MaterialExpressionTextureSample_10
       inputName=None
       mask=1
       maskA=0
       maskB=0
       maskG=0
       maskR=1
       outputIndex=1
      }
     }
     {
      type=Scalar
      expressionInput {
       expression=&self:MaterialExpressionMultiply_34
       inputName=None
       mask=0
       maskA=0
       maskB=0
       maskG=0
       maskR=0
       outputIndex=0
      }
     }
     {
      type=Scalar
      expressionInput {
       expression=&self:MaterialExpressionTextureSample_10
       inputName=None
       mask=1
       maskA=0
       maskB=1
       maskG=0
       maskR=0
       outputIndex=3
      }
     }
    ]
    convertOutputs [{type=Vector3}]
   }
  }
  MaterialExpressionConvert_2 {
   class=MaterialExpressionConvert
   inputs [
    None {from=MaterialExpressionTextureSample_11, output=R}
    None {from=MaterialExpressionMultiply_35}
    None {from=MaterialExpressionTextureSample_11, output=R}
   ]
   properties {
    nodeName=Make Vector3
    convertInputs [
     {
      type=Scalar
      expressionInput {
       expression=&self:MaterialExpressionTextureSample_11
       inputName=None
       mask=1
       maskA=0
       maskB=0
       maskG=0
       maskR=1
       outputIndex=1
      }
     }
     {
      type=Scalar
      expressionInput {
       expression=&self:MaterialExpressionMultiply_35
       inputName=None
       mask=0
       maskA=0
       maskB=0
       maskG=0
       maskR=0
       outputIndex=0
      }
     }
     {
      type=Scalar
      expressionInput {
       expression=&self:MaterialExpressionTextureSample_11
       inputName=None
       mask=1
       maskA=0
       maskB=1
       maskG=0
       maskR=0
       outputIndex=3
      }
     }
    ]
    convertOutputs [{type=Vector3}]
   }
  }
  MaterialExpressionConvert_3 {
   class=MaterialExpressionConvert
   inputs [
    None {from=MaterialExpressionTextureSample_15, output=R}
    None {from=MaterialExpressionMultiply_36}
    None {from=MaterialExpressionTextureSample_15, output=R}
   ]
   properties {
    nodeName=Make Vector3
    convertInputs [
     {
      type=Scalar
      expressionInput {
       expression=&self:MaterialExpressionTextureSample_15
       inputName=None
       mask=1
       maskA=0
       maskB=0
       maskG=0
       maskR=1
       outputIndex=1
      }
     }
     {
      type=Scalar
      expressionInput {
       expression=&self:MaterialExpressionMultiply_36
       inputName=None
       mask=0
       maskA=0
       maskB=0
       maskG=0
       maskR=0
       outputIndex=0
      }
     }
     {
      type=Scalar
      expressionInput {
       expression=&self:MaterialExpressionTextureSample_15
       inputName=None
       mask=1
       maskA=0
       maskB=1
       maskG=0
       maskR=0
       outputIndex=3
      }
     }
    ]
    convertOutputs [{type=Vector3}]
   }
  }
  MaterialExpressionCustom_3 {
   class=MaterialExpressionCustom
   inputs [
    v {from=MaterialExpressionMultiply_28, output=""}
    n {from=MaterialExpressionNormalize_1, output=""}
    i {from=MaterialExpressionScalarParameter_26, output=""}
   ]
   properties {
    code="return refract(v, n, i);"
    description=refract
    inputs [
     {
      inputName=v
      input {
       expression=&self:MaterialExpressionMultiply_28
       inputName=None
       mask=0
       maskA=0
       maskB=0
       maskG=0
       maskR=0
       outputIndex=0
      }
     }
     {
      inputName=n
      input {
       expression=&self:MaterialExpressionNormalize_1
       inputName=None
       mask=0
       maskA=0
       maskB=0
       maskG=0
       maskR=0
       outputIndex=0
      }
     }
     {
      inputName=i
      input {
       expression=&self:MaterialExpressionScalarParameter_26
       inputName=None
       mask=0
       maskA=0
       maskB=0
       maskG=0
       maskR=0
       outputIndex=0
      }
     }
    ]
   }
  }
  MaterialExpressionDDX_0 {class=MaterialExpressionDDX, inputs [Value {from=MaterialExpressionComponentMask_4, output=""}], properties {}}
  MaterialExpressionDDX_1 {class=MaterialExpressionDDX, inputs [Value {from=MaterialExpressionComponentMask_6, output=""}], properties {}}
  MaterialExpressionDDX_10 {class=MaterialExpressionDDX, inputs [Value {from=MaterialExpressionComponentMask_26, output=""}], properties {}}
  MaterialExpressionDDX_2 {class=MaterialExpressionDDX, inputs [Value {from=MaterialExpressionComponentMask_9, output=""}], properties {}}
  MaterialExpressionDDX_3 {class=MaterialExpressionDDX, inputs [Value {from=MaterialExpressionComponentMask_11, output=""}], properties {}}
  MaterialExpressionDDX_4 {class=MaterialExpressionDDX, inputs [Value {from=MaterialExpressionComponentMask_14, output=""}], properties {}}
  MaterialExpressionDDX_5 {class=MaterialExpressionDDX, inputs [Value {from=MaterialExpressionComponentMask_16, output=""}], properties {}}
  MaterialExpressionDDX_6 {class=MaterialExpressionDDX, inputs [Value {from=MaterialExpressionComponentMask_18, output=""}], properties {}}
  MaterialExpressionDDX_7 {class=MaterialExpressionDDX, inputs [Value {from=MaterialExpressionComponentMask_20, output=""}], properties {}}
  MaterialExpressionDDX_8 {class=MaterialExpressionDDX, inputs [Value {from=MaterialExpressionComponentMask_22, output=""}], properties {}}
  MaterialExpressionDDX_9 {class=MaterialExpressionDDX, inputs [Value {from=MaterialExpressionComponentMask_24, output=""}], properties {}}
  MaterialExpressionDDY_0 {class=MaterialExpressionDDY, inputs [Value {from=MaterialExpressionComponentMask_5, output=""}], properties {}}
  MaterialExpressionDDY_1 {class=MaterialExpressionDDY, inputs [Value {from=MaterialExpressionComponentMask_8, output=""}], properties {}}
  MaterialExpressionDDY_10 {class=MaterialExpressionDDY, inputs [Value {from=MaterialExpressionComponentMask_27, output=""}], properties {}}
  MaterialExpressionDDY_2 {class=MaterialExpressionDDY, inputs [Value {from=MaterialExpressionComponentMask_10, output=""}], properties {}}
  MaterialExpressionDDY_3 {class=MaterialExpressionDDY, inputs [Value {from=MaterialExpressionComponentMask_13, output=""}], properties {}}
  MaterialExpressionDDY_4 {class=MaterialExpressionDDY, inputs [Value {from=MaterialExpressionComponentMask_15, output=""}], properties {}}
  MaterialExpressionDDY_5 {class=MaterialExpressionDDY, inputs [Value {from=MaterialExpressionComponentMask_17, output=""}], properties {}}
  MaterialExpressionDDY_6 {class=MaterialExpressionDDY, inputs [Value {from=MaterialExpressionComponentMask_19, output=""}], properties {}}
  MaterialExpressionDDY_7 {class=MaterialExpressionDDY, inputs [Value {from=MaterialExpressionComponentMask_21, output=""}], properties {}}
  MaterialExpressionDDY_8 {class=MaterialExpressionDDY, inputs [Value {from=MaterialExpressionComponentMask_23, output=""}], properties {}}
  MaterialExpressionDDY_9 {class=MaterialExpressionDDY, inputs [Value {from=MaterialExpressionComponentMask_25, output=""}], properties {}}
  MaterialExpressionDecalLifetimeOpacity_1 {class=MaterialExpressionDecalLifetimeOpacity, inputs [], properties {}}
  MaterialExpressionDivide_2 {
   class=MaterialExpressionDivide
   inputs [
    A {from=MaterialExpressionMax_1}
    B {from=MaterialExpressionSceneDepth_1, output=R}
   ]
  }
  MaterialExpressionLightVector_0 {class=MaterialExpressionLightVector, inputs [], properties {}}
  MaterialExpressionLinearInterpolate_0 {
   class=MaterialExpressionLinearInterpolate
   inputs [
    A {from=""}
    B {from=""}
    Alpha {from=MaterialExpressionComponentMask_0}
   ]
   properties {constB=0.5}
  }
  MaterialExpressionLinearInterpolate_1 {
   class=MaterialExpressionLinearInterpolate
   inputs [
    A {from=MaterialExpressionVectorParameter_0, output=RGB}
    B {from=MaterialExpressionVectorParameter_4, output=RGB}
    Alpha {from=MaterialExpressionSaturate_0}
   ]
  }
  MaterialExpressionLinearInterpolate_2 {
   class=MaterialExpressionLinearInterpolate
   inputs [
    A {from=MaterialExpressionReroute_29}
    B {from=""}
    Alpha {from=MaterialExpressionSaturate_1}
   ]
  }
  MaterialExpressionLinearInterpolate_4 {
   class=MaterialExpressionLinearInterpolate
   inputs [
    A {from=""}
    B {from=""}
    Alpha {from=MaterialExpressionComponentMask_0}
   ]
   properties {constA=1}
  }
  MaterialExpressionMaterialFunctionCall_0 {
   class=MaterialExpressionMaterialFunctionCall
   inputs [
    UVs {from=MaterialExpressionTextureCoordinate_5}
    Additional Scale {from=""}
   ]
   properties {materialFunction=/Game/Effects/Fort_Effects/Materials/_Functions/MF_DecalSphereCoords}
  }
  MaterialExpressionMaterialFunctionCall_1 {
   class=MaterialExpressionMaterialFunctionCall
   inputs [
    Dynamic Param {from=MaterialExpressionMultiply_16}
    Alpha {from=MaterialExpressionMultiply_6}
   ]
   properties {materialFunction=/Game/Effects/Fort_Effects/Materials/_Functions/Erode_Alpha}
  }
  MaterialExpressionMaterialFunctionCall_10 {
   class=MaterialExpressionMaterialFunctionCall
   inputs [
    UVs {from=MaterialExpressionTextureCoordinate_3}
    Additional Scale {from=""}
   ]
   properties {materialFunction=&MF_DecalSphereCoords}
  }
  MaterialExpressionMaterialFunctionCall_11 {
   class=MaterialExpressionMaterialFunctionCall
   inputs [
    UVs {from=MaterialExpressionTextureCoordinate_6}
    Additional Scale {from=""}
   ]
   properties {materialFunction=&MF_DecalSphereCoords}
  }
  MaterialExpressionMaterialFunctionCall_12 {
   class=MaterialExpressionMaterialFunctionCall
   inputs [
    UVs {from=MaterialExpressionTextureCoordinate_4}
    Additional Scale {from=""}
   ]
   properties {materialFunction=&MF_DecalSphereCoords}
  }
  MaterialExpressionMaterialFunctionCall_13 {
   class=MaterialExpressionMaterialFunctionCall
   inputs [
    BaseNormal {from=MaterialExpressionTextureSample_0, output=RGB}
    AdditionalNormal {from=MaterialExpressionTextureSample_1, output=RGB}
   ]
   properties {materialFunction=&BlendAngleCorrectedNormals}
  }
  MaterialExpressionMaterialFunctionCall_14 {
   class=MaterialExpressionMaterialFunctionCall
   inputs [
    BaseNormal {from=MaterialExpressionMaterialFunctionCall_16, output=Result}
    AdditionalNormal {from=MaterialExpressionMaterialFunctionCall_2, output=Result}
   ]
   properties {materialFunction=&BlendAngleCorrectedNormals}
  }
  MaterialExpressionMaterialFunctionCall_15 {
   class=MaterialExpressionMaterialFunctionCall
   inputs [Speed {from=MaterialExpressionScalarParameter_42}]
   properties {materialFunction=&DebugTimeSine}
  }
  MaterialExpressionMaterialFunctionCall_16 {
   class=MaterialExpressionMaterialFunctionCall
   inputs [
    Normal {from=MaterialExpressionTextureSample_12, output=RGB}
    Flatness {from=MaterialExpressionScalarParameter_25}
   ]
   properties {materialFunction=&FlattenNormal}
  }
  MaterialExpressionMaterialFunctionCall_17 {
   class=MaterialExpressionMaterialFunctionCall
   inputs [
    BaseNormal {from=MaterialExpressionTextureSample_18, output=RGB}
    AdditionalNormal {from=MaterialExpressionTextureSample_25, output=RGB}
   ]
   properties {materialFunction=&BlendAngleCorrectedNormals}
  }
  MaterialExpressionMaterialFunctionCall_18 {
   class=MaterialExpressionMaterialFunctionCall
   inputs [
    Normal {from=MaterialExpressionNormalize_5}
    Flatness {from=MaterialExpressionScalarParameter_30}
   ]
   properties {materialFunction=&FlattenNormal}
  }
  MaterialExpressionMaterialFunctionCall_19 {
   class=MaterialExpressionMaterialFunctionCall
   inputs [
    UVs {from=MaterialExpressionTextureCoordinate_7}
    Additional Scale {from=""}
   ]
   properties {materialFunction=&MF_DecalSphereCoords}
  }
  MaterialExpressionMaterialFunctionCall_2 {
   class=MaterialExpressionMaterialFunctionCall
   inputs [
    Normal {from=MaterialExpressionMaterialFunctionCall_13, output=Result}
    Flatness {from=MaterialExpressionScalarParameter_6}
   ]
   properties {materialFunction=&FlattenNormal}
  }
  MaterialExpressionMaterialFunctionCall_20 {
   class=MaterialExpressionMaterialFunctionCall
   inputs [
    UVs {from=MaterialExpressionTextureCoordinate_8}
    Additional Scale {from=""}
   ]
   properties {materialFunction=&MF_DecalSphereCoords}
  }
  MaterialExpressionMaterialFunctionCall_3 {
   class=MaterialExpressionMaterialFunctionCall
   inputs [
    WorldPosition {from=MaterialExpressionAdd_12}
    Scale {from=MaterialExpressionScalarParameter_23}
    Water Depth {from=MaterialExpressionReroute_13}
    Time {from=MaterialExpressionMaterialFunctionCall_15, output=Result}
    Water Velocity {from=MaterialExpressionComponentMask_12}
   ]
   properties {materialFunction=/FortWater/Materials/Functions/WaterCaustics}
  }
  MaterialExpressionMaterialFunctionCall_4 {
   class=MaterialExpressionMaterialFunctionCall
   inputs [Screen Position {from=MaterialExpressionAdd_29}]
   properties {desc=Dont pull in pixels from above the water, materialFunction=/FortWater/Materials/Functions/Distortion_MaskForeground}
  }
  MaterialExpressionMaterialFunctionCall_5 {
   class=MaterialExpressionMaterialFunctionCall
   inputs [
    SceneDepth {from=MaterialExpressionSceneDepth_2, output=R}
    Exclude WPO Offsets {from=""}
   ]
   properties {materialFunction=&WorldPositionBehindTranslucency}
  }
  MaterialExpressionMaterialFunctionCall_6 {
   class=MaterialExpressionMaterialFunctionCall
   inputs [
    UVs {from=MaterialExpressionTextureCoordinate_1}
    Additional Scale {from=""}
   ]
   properties {materialFunction=&MF_DecalSphereCoords}
  }
  MaterialExpressionMaterialFunctionCall_7 {
   class=MaterialExpressionMaterialFunctionCall
   inputs [
    BaseNormal {from=MaterialExpressionReroute_23}
    AdditionalNormal {from=MaterialExpressionNormalize_0}
   ]
   properties {materialFunction=&BlendAngleCorrectedNormals}
  }
  MaterialExpressionMaterialFunctionCall_8 {
   class=MaterialExpressionMaterialFunctionCall
   inputs [
    UVs {from=MaterialExpressionTextureCoordinate_2}
    Additional Scale {from=""}
   ]
   properties {materialFunction=&MF_DecalSphereCoords}
  }
  MaterialExpressionMaterialFunctionCall_86 {
   class=MaterialExpressionMaterialFunctionCall
   inputs [
    World Normal {from=MaterialExpressionMultiply_19}
    Custom UV Input {from=""}
   ]
   properties {materialFunction=/FortWater/Materials/Functions/Transform_Refraction_Normals}
  }
  MaterialExpressionMaterialFunctionCall_9 {
   class=MaterialExpressionMaterialFunctionCall
   inputs [
    UVs {from=MaterialExpressionTextureCoordinate_13}
    Additional Scale {from=""}
   ]
   properties {materialFunction=&MF_DecalSphereCoords}
  }
  MaterialExpressionMax_1 {
   class=MaterialExpressionMax
   inputs [
    A {from=MaterialExpressionSubtract_2}
    B {from=MaterialExpressionScalarParameter_93}
   ]
  }
  MaterialExpressionMultiply_0 {
   class=MaterialExpressionMultiply
   inputs [
    A {from=MaterialExpressionReroute_28}
    B {from=MaterialExpressionComponentMask_2}
   ]
  }
  MaterialExpressionMultiply_1 {
   class=MaterialExpressionMultiply
   inputs [
    A {from=MaterialExpressionMaterialFunctionCall_0, output=Result}
    B {from=""}
   ]
   properties {constB=0.5}
  }
  MaterialExpressionMultiply_10 {
   class=MaterialExpressionMultiply
   inputs [
    A {from=MaterialExpressionTextureCoordinate_1}
    B {from=""}
   ]
   properties {constB=0.5}
  }
  MaterialExpressionMultiply_11 {
   class=MaterialExpressionMultiply
   inputs [
    A {from=MaterialExpressionMaterialFunctionCall_8, output=Result}
    B {from=""}
   ]
   properties {constB=0.5}
  }
  MaterialExpressionMultiply_12 {
   class=MaterialExpressionMultiply
   inputs [
    A {from=MaterialExpressionTextureCoordinate_2}
    B {from=""}
   ]
   properties {constB=0.5}
  }
  MaterialExpressionMultiply_13 {
   class=MaterialExpressionMultiply
   inputs [
    A {from=MaterialExpressionMaterialFunctionCall_10, output=Result}
    B {from=""}
   ]
   properties {constB=0.5}
  }
  MaterialExpressionMultiply_14 {
   class=MaterialExpressionMultiply
   inputs [
    A {from=MaterialExpressionTextureCoordinate_3}
    B {from=""}
   ]
   properties {constB=0.5}
  }
  MaterialExpressionMultiply_15 {
   class=MaterialExpressionMultiply
   inputs [
    A {from=MaterialExpressionMaterialFunctionCall_11, output=Result}
    B {from=""}
   ]
   properties {constB=0.5}
  }
  MaterialExpressionMultiply_16 {
   class=MaterialExpressionMultiply
   inputs [
    A {from=MaterialExpressionOneMinus_1}
    B {from=MaterialExpressionScalarParameter_3}
   ]
  }
  MaterialExpressionMultiply_17 {
   class=MaterialExpressionMultiply
   inputs [
    A {from=MaterialExpressionTextureCoordinate_6}
    B {from=""}
   ]
   properties {constB=0.5}
  }
  MaterialExpressionMultiply_18 {
   class=MaterialExpressionMultiply
   inputs [
    A {from=MaterialExpressionMaterialFunctionCall_19, output=Result}
    B {from=""}
   ]
   properties {constB=0.5}
  }
  MaterialExpressionMultiply_19 {
   class=MaterialExpressionMultiply
   inputs [
    A {from=MaterialExpressionMultiply_22}
    B {from=MaterialExpressionMultiply_20}
   ]
  }
  MaterialExpressionMultiply_2 {
   class=MaterialExpressionMultiply
   inputs [
    A {from=MaterialExpressionMaterialFunctionCall_6, output=Result}
    B {from=""}
   ]
   properties {constB=0.5}
  }
  MaterialExpressionMultiply_20 {
   class=MaterialExpressionMultiply
   inputs [
    A {from=MaterialExpressionDivide_2}
    B {from=MaterialExpressionScalarParameter_47}
   ]
  }
  MaterialExpressionMultiply_21 {
   class=MaterialExpressionMultiply
   inputs [
    A {from=MaterialExpressionTextureCoordinate_7}
    B {from=""}
   ]
   properties {constB=0.5}
  }
  MaterialExpressionMultiply_22 {
   class=MaterialExpressionMultiply
   inputs [
    A {from=MaterialExpressionReroute_11}
    B {from=""}
   ]
   properties {constA=1, desc=Normal}
  }
  MaterialExpressionMultiply_23 {
   class=MaterialExpressionMultiply
   inputs [
    A {from=MaterialExpressionReroute_12}
    B {from=MaterialExpressionScalarParameter_24}
   ]
   properties {desc=Refracted Vector * Depth}
  }
  MaterialExpressionMultiply_24 {
   class=MaterialExpressionMultiply
   inputs [
    A {from=MaterialExpressionMaterialFunctionCall_12, output=Result}
    B {from=""}
   ]
   properties {constB=0.5}
  }
  MaterialExpressionMultiply_25 {
   class=MaterialExpressionMultiply
   inputs [
    A {from=MaterialExpressionTextureCoordinate_4}
    B {from=""}
   ]
   properties {constB=0.5}
  }
  MaterialExpressionMultiply_26 {
   class=MaterialExpressionMultiply
   inputs [
    A {from=MaterialExpressionMaterialFunctionCall_20, output=Result}
    B {from=""}
   ]
  }
  MaterialExpressionMultiply_27 {
   class=MaterialExpressionMultiply
   inputs [
    A {from=MaterialExpressionMultiply_23}
    B {from=MaterialExpressionScalarParameter_22}
   ]
  }
  MaterialExpressionMultiply_28 {
   class=MaterialExpressionMultiply
   inputs [
    A {from=MaterialExpressionCameraVectorWS_0}
    B {from=""}
   ]
   properties {constB=-1}
  }
  MaterialExpressionMultiply_29 {
   class=MaterialExpressionMultiply
   inputs [
    A {from=MaterialExpressionReroute_25}
    B {from=MaterialExpressionComponentMask_1}
   ]
   properties {constB=0.8556}
  }
  MaterialExpressionMultiply_3 {
   class=MaterialExpressionMultiply
   inputs [
    A {from=MaterialExpressionScalarParameter_2}
    B {from=MaterialExpressionDecalLifetimeOpacity_1, output=Opacity}
   ]
  }
  MaterialExpressionMultiply_30 {
   class=MaterialExpressionMultiply
   inputs [
    A {from=MaterialExpressionVectorParameter_12, output=RGB}
    B {from=MaterialExpressionReroute_11}
   ]
  }
  MaterialExpressionMultiply_31 {
   class=MaterialExpressionMultiply
   inputs [
    A {from=MaterialExpressionReroute_15}
    B {from=MaterialExpressionVectorParameter_13, output=RGB}
   ]
  }
  MaterialExpressionMultiply_32 {
   class=MaterialExpressionMultiply
   inputs [
    A {from=MaterialExpressionComponentMask_43}
    B {from=MaterialExpressionScalarParameter_27}
   ]
   properties {constB=0.04}
  }
  MaterialExpressionMultiply_33 {
   class=MaterialExpressionMultiply
   inputs [
    A {from=MaterialExpressionTextureCoordinate_8}
    B {from=""}
   ]
  }
  MaterialExpressionMultiply_34 {
   class=MaterialExpressionMultiply
   inputs [
    A {from=MaterialExpressionTextureSample_10, output=G}
    B {from=""}
   ]
   properties {constB=-1}
  }
  MaterialExpressionMultiply_35 {
   class=MaterialExpressionMultiply
   inputs [
    A {from=MaterialExpressionTextureSample_11, output=G}
    B {from=""}
   ]
   properties {constB=-1}
  }
  MaterialExpressionMultiply_36 {
   class=MaterialExpressionMultiply
   inputs [
    A {from=MaterialExpressionTextureSample_15, output=G}
    B {from=""}
   ]
   properties {constB=-1}
  }
  MaterialExpressionMultiply_4 {
   class=MaterialExpressionMultiply
   inputs [
    A {from=MaterialExpressionTextureCoordinate_14}
    B {from=MaterialExpressionComponentMask_3}
   ]
   properties {constB=0.8556}
  }
  MaterialExpressionMultiply_5 {
   class=MaterialExpressionMultiply
   inputs [
    A {from=MaterialExpressionTextureCoordinate_0}
    B {from=MaterialExpressionComponentMask_7}
   ]
  }
  MaterialExpressionMultiply_6 {
   class=MaterialExpressionMultiply
   inputs [
    A {from=MaterialExpressionAdd_11}
    B {from=MaterialExpressionSphereMask_0}
   ]
  }
  MaterialExpressionMultiply_7 {
   class=MaterialExpressionMultiply
   inputs [
    A {from=MaterialExpressionTextureSample_14, output=G}
    B {from=""}
   ]
   properties {constB=-1}
  }
  MaterialExpressionMultiply_8 {
   class=MaterialExpressionMultiply
   inputs [
    A {from=MaterialExpressionTextureCoordinate_5}
    B {from=""}
   ]
   properties {constB=0.5}
  }
  MaterialExpressionMultiply_9 {
   class=MaterialExpressionMultiply
   inputs [
    A {from=MaterialExpressionMaterialFunctionCall_1, output=Result}
    B {from=MaterialExpressionScalarParameter_5}
   ]
  }
  MaterialExpressionNormalize_0 {class=MaterialExpressionNormalize, inputs [VectorInput {from=MaterialExpressionAdd_13, output=""}]}
  MaterialExpressionNormalize_1 {class=MaterialExpressionNormalize, inputs [VectorInput {from=MaterialExpressionMultiply_30, output=""}]}
  MaterialExpressionNormalize_5 {class=MaterialExpressionNormalize, inputs [VectorInput {from=MaterialExpressionMaterialFunctionCall_17, output=Result}]}
  MaterialExpressionOneMinus_1 {class=MaterialExpressionOneMinus, inputs [None {from=MaterialExpressionMultiply_3}]}
  MaterialExpressionPixelDepth_1 {class=MaterialExpressionPixelDepth, inputs [], properties {}}
  MaterialExpressionPower_5 {
   class=MaterialExpressionPower
   inputs [
    Base {from=MaterialExpressionComponentMask_0}
    Exp {from=MaterialExpressionConstant_1}
   ]
  }
  MaterialExpressionReroute_0 {class=MaterialExpressionReroute, inputs [None {from=MaterialExpressionAdd_5}]}
  MaterialExpressionReroute_1 {class=MaterialExpressionReroute, inputs [None {from=MaterialExpressionReroute_0}]}
  MaterialExpressionReroute_11 {class=MaterialExpressionReroute, inputs [None {from=MaterialExpressionReroute_23}]}
  MaterialExpressionReroute_12 {class=MaterialExpressionReroute, inputs [None {from=MaterialExpressionCustom_3}]}
  MaterialExpressionReroute_13 {class=MaterialExpressionReroute, inputs [None {from=MaterialExpressionScalarParameter_24}]}
  MaterialExpressionReroute_14 {class=MaterialExpressionReroute, inputs [None {from=MaterialExpressionLinearInterpolate_1}]}
  MaterialExpressionReroute_15 {class=MaterialExpressionReroute, inputs [None {from=MaterialExpressionMaterialFunctionCall_3, output=Result}]}
  MaterialExpressionReroute_22 {class=MaterialExpressionReroute, inputs [None {from=MaterialExpressionShadingPathSwitch_0}]}
  MaterialExpressionReroute_23 {class=MaterialExpressionReroute, inputs [None {from=MaterialExpressionReroute_22}]}
  MaterialExpressionReroute_24 {class=MaterialExpressionReroute, inputs [None {from=MaterialExpressionMaterialFunctionCall_9, output=Result}]}
  MaterialExpressionReroute_25 {class=MaterialExpressionReroute, inputs [None {from=MaterialExpressionReroute_24}]}
  MaterialExpressionReroute_26 {class=MaterialExpressionReroute, inputs [None {from=MaterialExpressionReroute_14}]}
  MaterialExpressionReroute_27 {class=MaterialExpressionReroute, inputs [None {from=MaterialExpressionReroute_24}]}
  MaterialExpressionReroute_28 {class=MaterialExpressionReroute, inputs [None {from=MaterialExpressionReroute_27}]}
  MaterialExpressionReroute_29 {class=MaterialExpressionReroute, inputs [None {from=MaterialExpressionLinearInterpolate_4}]}
  MaterialExpressionReroute_30 {class=MaterialExpressionReroute, inputs [None {from=MaterialExpressionLinearInterpolate_0}]}
  MaterialExpressionReroute_8 {class=MaterialExpressionReroute, inputs [None {from=MaterialExpressionAdd_2}]}
  MaterialExpressionSaturate_0 {class=MaterialExpressionSaturate, inputs [None {from=MaterialExpressionPower_5}]}
  MaterialExpressionSaturate_1 {class=MaterialExpressionSaturate, inputs [None {from=MaterialExpressionSubtract_0}]}
  MaterialExpressionScalarParameter_1 {class=MaterialExpressionScalarParameter, properties {defaultValue=0.2, parameterName=B_1}}
  MaterialExpressionScalarParameter_2 {class=MaterialExpressionScalarParameter, properties {defaultValue=0.9, parameterName=FadeIn}}
  MaterialExpressionScalarParameter_22 {class=MaterialExpressionScalarParameter, properties {defaultValue=-1, group=Caustics, parameterName=Caustic Refraction}}
  MaterialExpressionScalarParameter_23 {class=MaterialExpressionScalarParameter, properties {defaultValue=512, group=Caustics, parameterName=Caustics Scale}}
  MaterialExpressionScalarParameter_24 {class=MaterialExpressionScalarParameter, properties {defaultValue=150, group=Caustics, parameterName=VatDepth}}
  MaterialExpressionScalarParameter_25 {
   class=MaterialExpressionScalarParameter
   properties {defaultValue=0.9115, group=Normals_Mobile, parameterName=NormalFlaten_M3, sortPriority=5}
  }
  MaterialExpressionScalarParameter_26 {class=MaterialExpressionScalarParameter, properties {defaultValue=0.75, group=Caustics, parameterName=Relative Refractive Indices}}
  MaterialExpressionScalarParameter_27 {class=MaterialExpressionScalarParameter, properties {defaultValue=0.1, group=Normals, parameterName=Noise 1}}
  MaterialExpressionScalarParameter_3 {class=MaterialExpressionScalarParameter, properties {defaultValue=-4.5, parameterName=ErodeStrength}}
  MaterialExpressionScalarParameter_30 {class=MaterialExpressionScalarParameter, properties {group=Normals, parameterName=NormalFlatten}}
  MaterialExpressionScalarParameter_42 {class=MaterialExpressionScalarParameter, properties {defaultValue=0.1, parameterName=Time}}
  MaterialExpressionScalarParameter_47 {class=MaterialExpressionScalarParameter, properties {defaultValue=-1, desc=-0.4, group=Caustics, parameterName=Pixel Normal Offset}}
  MaterialExpressionScalarParameter_5 {class=MaterialExpressionScalarParameter, properties {defaultValue=0.8, parameterName=Opacity}}
  MaterialExpressionScalarParameter_6 {class=MaterialExpressionScalarParameter, properties {group=Normals, parameterName=NormalFlatten}}
  MaterialExpressionScalarParameter_93 {class=MaterialExpressionScalarParameter, properties {group=Caustics, parameterName=Min Refraction and Blur}}
  MaterialExpressionSceneDepth_1 {class=MaterialExpressionSceneDepth, inputs [UVs {from="", output=""}], properties {}}
  MaterialExpressionSceneDepth_2 {class=MaterialExpressionSceneDepth, inputs [UVs {from=MaterialExpressionMaterialFunctionCall_4, output=Result}], properties {}}
  MaterialExpressionScreenPosition_13 {class=MaterialExpressionScreenPosition, inputs [], properties {}}
  MaterialExpressionShadingPathSwitch_0 {
   class=MaterialExpressionShadingPathSwitch
   inputs [
    Default {from=MaterialExpressionMaterialFunctionCall_18, output=Result}
    Deferred {from="", output=""}
    Forward {from=MaterialExpressionMaterialFunctionCall_14, output=Result}
    Mobile {from=MaterialExpressionMaterialFunctionCall_14, output=Result}
   ]
   properties {}
  }
  MaterialExpressionShadingPathSwitch_1 {
   class=MaterialExpressionShadingPathSwitch
   inputs [
    Default {from=MaterialExpressionTextureSample_5, output=G}
    Deferred {from="", output=""}
    Forward {from=MaterialExpressionTextureSample_13, output=G}
    Mobile {from=MaterialExpressionTextureSample_13, output=G}
   ]
   properties {}
  }
  MaterialExpressionShadingPathSwitch_10 {
   class=MaterialExpressionShadingPathSwitch
   inputs [
    Default {from=MaterialExpressionLightVector_0, output=""}
    Deferred {from="", output=""}
    Forward {from=MaterialExpressionTextureCoordinate_9, output=""}
    Mobile {from=MaterialExpressionTextureCoordinate_9, output=""}
   ]
   properties {}
  }
  MaterialExpressionShadingPathSwitch_2 {
   class=MaterialExpressionShadingPathSwitch
   inputs [
    Default {from=MaterialExpressionTextureSample_6, output=G}
    Deferred {from="", output=""}
    Forward {from=MaterialExpressionTextureSample_19, output=G}
    Mobile {from=MaterialExpressionTextureSample_19, output=G}
   ]
   properties {}
  }
  MaterialExpressionShadingPathSwitch_3 {
   class=MaterialExpressionShadingPathSwitch
   inputs [
    Default {from=MaterialExpressionTextureSample_16, output=RGB}
    Deferred {from="", output=""}
    Forward {from=MaterialExpressionTextureSample_4, output=RGB}
    Mobile {from=MaterialExpressionTextureSample_4, output=RGB}
   ]
   properties {}
  }
  MaterialExpressionShadingPathSwitch_4 {
   class=MaterialExpressionShadingPathSwitch
   inputs [
    Default {from=MaterialExpressionTextureSample_7, output=RGB}
    Deferred {from="", output=""}
    Forward {from=MaterialExpressionTextureSample_17, output=RGB}
    Mobile {from=MaterialExpressionTextureSample_17, output=RGB}
   ]
   properties {}
  }
  MaterialExpressionShadingPathSwitch_5 {
   class=MaterialExpressionShadingPathSwitch
   inputs [
    Default {from=MaterialExpressionConvert_0, output=""}
    Deferred {from="", output=""}
    Forward {from=MaterialExpressionConvert_1, output=""}
    Mobile {from=MaterialExpressionConvert_1, output=""}
   ]
   properties {}
  }
  MaterialExpressionShadingPathSwitch_6 {
   class=MaterialExpressionShadingPathSwitch
   inputs [
    Default {from=MaterialExpressionConvert_2, output=""}
    Deferred {from="", output=""}
    Forward {from=MaterialExpressionConvert_3, output=""}
    Mobile {from=MaterialExpressionConvert_3, output=""}
   ]
   properties {}
  }
  MaterialExpressionShadingPathSwitch_7 {
   class=MaterialExpressionShadingPathSwitch
   inputs [
    Default {from=MaterialExpressionTextureSample_8, output=G}
    Deferred {from="", output=""}
    Forward {from=MaterialExpressionTextureSample_2, output=G}
    Mobile {from=MaterialExpressionTextureSample_2, output=G}
   ]
   properties {}
  }
  MaterialExpressionShadingPathSwitch_8 {
   class=MaterialExpressionShadingPathSwitch
   inputs [
    Default {from=MaterialExpressionTextureSample_3, output=G}
    Deferred {from="", output=""}
    Forward {from=MaterialExpressionTextureSample_9, output=G}
    Mobile {from=MaterialExpressionTextureSample_9, output=G}
   ]
   properties {}
  }
  MaterialExpressionShadingPathSwitch_9 {
   class=MaterialExpressionShadingPathSwitch
   inputs [
    Default {from=MaterialExpressionMultiply_31, output=""}
    Deferred {from="", output=""}
    Forward {from=MaterialExpressionVectorParameter_13, output=RGB}
    Mobile {from=MaterialExpressionVectorParameter_13, output=RGB}
   ]
   properties {}
  }
  MaterialExpressionSphereMask_0 {
   class=MaterialExpressionSphereMask
   inputs [
    A {from=MaterialExpressionShadingPathSwitch_10, output=""}
    B {from=MaterialExpressionConstant_3, output=""}
    Radius {from=MaterialExpressionConstant_2, output=""}
    Hardness {from="", output=""}
   ]
   properties {hardnessPercent=90}
  }
  MaterialExpressionSubtract_0 {
   class=MaterialExpressionSubtract
   inputs [
    A {from=MaterialExpressionViewProperty_0, output=Property}
    B {from=""}
   ]
   properties {constB=100}
  }
  MaterialExpressionSubtract_2 {
   class=MaterialExpressionSubtract
   inputs [
    A {from=MaterialExpressionSceneDepth_1, output=R}
    B {from=MaterialExpressionPixelDepth_1, output=R}
   ]
  }
  MaterialExpressionTextureCoordinate_0 {class=MaterialExpressionTextureCoordinate}
  MaterialExpressionTextureCoordinate_1 {class=MaterialExpressionTextureCoordinate}
  MaterialExpressionTextureCoordinate_13 {class=MaterialExpressionTextureCoordinate}
  MaterialExpressionTextureCoordinate_14 {class=MaterialExpressionTextureCoordinate}
  MaterialExpressionTextureCoordinate_2 {class=MaterialExpressionTextureCoordinate}
  MaterialExpressionTextureCoordinate_3 {class=MaterialExpressionTextureCoordinate}
  MaterialExpressionTextureCoordinate_4 {class=MaterialExpressionTextureCoordinate}
  MaterialExpressionTextureCoordinate_5 {class=MaterialExpressionTextureCoordinate}
  MaterialExpressionTextureCoordinate_6 {class=MaterialExpressionTextureCoordinate}
  MaterialExpressionTextureCoordinate_7 {class=MaterialExpressionTextureCoordinate}
  MaterialExpressionTextureCoordinate_8 {class=MaterialExpressionTextureCoordinate}
  MaterialExpressionTextureCoordinate_9 {class=MaterialExpressionTextureCoordinate}
  MaterialExpressionTextureSample_0 {
   class=MaterialExpressionTextureSample
   inputs [
    UVs {from=MaterialExpressionVertexInterpolator_0, output=PS}
    Tex {from=""}
    "DDX(UVs)" {from=MaterialExpressionDDX_6}
    "DDY(UVs)" {from=MaterialExpressionDDY_6}
    Apply View MipBias {from=""}
   ]
   properties {mipValueMode=TMVM_Derivative, samplerType=SAMPLERTYPE_Normal, texture=/Game/Global/Textures/Default/Tileable_Patterns/T_Water2_N}
  }
  MaterialExpressionTextureSample_1 {
   class=MaterialExpressionTextureSample
   inputs [
    UVs {from=MaterialExpressionVertexInterpolator_1, output=PS}
    Tex {from=""}
    "DDX(UVs)" {from=MaterialExpressionDDX_7}
    "DDY(UVs)" {from=MaterialExpressionDDY_7}
    Apply View MipBias {from=""}
   ]
   properties {mipValueMode=TMVM_Derivative, samplerType=SAMPLERTYPE_Normal, texture=&T_Water2_N}
  }
  MaterialExpressionTextureSample_10 {
   class=MaterialExpressionTextureSample
   inputs [
    UVs {from=MaterialExpressionAdd_10}
    Tex {from=""}
    "DDX(UVs)" {from=MaterialExpressionDDX_3}
    "DDY(UVs)" {from=MaterialExpressionDDY_3}
    Apply View MipBias {from=""}
   ]
   properties {
    mipValueMode=TMVM_Derivative
    samplerSource=SSM_Clamp_WorldGroupSettings
    samplerType=SAMPLERTYPE_Normal
    texture=/Game/Effects/Fort_Effects/Textures/Water/T_BoilBubble_Fix_N
   }
  }
  MaterialExpressionTextureSample_11 {
   class=MaterialExpressionTextureSample
   inputs [
    UVs {from=MaterialExpressionAdd_16}
    Tex {from=""}
    Apply View MipBias {from=""}
   ]
   properties {
    samplerSource=SSM_Clamp_WorldGroupSettings
    samplerType=SAMPLERTYPE_Normal
    texture=&T_BoilBubble_Fix_N
   }
  }
  MaterialExpressionTextureSample_12 {
   class=MaterialExpressionTextureSample
   inputs [
    UVs {from=MaterialExpressionVertexInterpolator_2, output=PS}
    Tex {from=""}
    "DDX(UVs)" {from=MaterialExpressionDDX_8}
    "DDY(UVs)" {from=MaterialExpressionDDY_8}
    Apply View MipBias {from=""}
   ]
   properties {mipValueMode=TMVM_Derivative, samplerType=SAMPLERTYPE_Normal, texture=/Game/Global/Textures/Default/Tileable_Patterns/T_WaterSoft_N}
  }
  MaterialExpressionTextureSample_13 {
   class=MaterialExpressionTextureSample
   inputs [
    UVs {from=MaterialExpressionMultiply_8}
    Tex {from=""}
    "DDX(UVs)" {from=MaterialExpressionDDX_5}
    "DDY(UVs)" {from=MaterialExpressionDDY_5}
    Apply View MipBias {from=""}
   ]
   properties {
    mipValueMode=TMVM_Derivative
    samplerSource=SSM_Clamp_WorldGroupSettings
    samplerType=SAMPLERTYPE_LinearColor
    texture=/Game/Effects/Fort_Effects/Textures/Water/T_waterSplash
   }
  }
  MaterialExpressionTextureSample_14 {
   class=MaterialExpressionTextureSample
   inputs [
    UVs {from=MaterialExpressionAdd_8}
    Tex {from=""}
    Level {from=""}
    Apply View MipBias {from=""}
   ]
   properties {
    constMipValue=0
    mipValueMode=TMVM_MipLevel
    samplerSource=SSM_Clamp_WorldGroupSettings
    samplerType=SAMPLERTYPE_Normal
    texture=&T_BoilBubble_Fix_N
   }
  }
  MaterialExpressionTextureSample_15 {
   class=MaterialExpressionTextureSample
   inputs [
    UVs {from=MaterialExpressionAdd_17}
    Tex {from=""}
    "DDX(UVs)" {from=MaterialExpressionDDX_4}
    "DDY(UVs)" {from=MaterialExpressionDDY_4}
    Apply View MipBias {from=""}
   ]
   properties {
    constMipValue=0
    mipValueMode=TMVM_Derivative
    samplerSource=SSM_Clamp_WorldGroupSettings
    samplerType=SAMPLERTYPE_Normal
    texture=&T_BoilBubble_Fix_N
   }
  }
  MaterialExpressionTextureSample_16 {
   class=MaterialExpressionTextureSample
   inputs [
    UVs {from=MaterialExpressionMultiply_11}
    Tex {from=""}
    Level {from=""}
    Apply View MipBias {from=""}
   ]
   properties {
    constMipValue=0
    mipValueMode=TMVM_MipLevel
    samplerSource=SSM_Clamp_WorldGroupSettings
    samplerType=SAMPLERTYPE_Normal
    texture=/Game/Effects/Fort_Effects/Textures/Water/T_waterSplash_N
   }
  }
  MaterialExpressionTextureSample_17 {
   class=MaterialExpressionTextureSample
   inputs [
    UVs {from=MaterialExpressionAdd_4}
    Tex {from=""}
    "DDX(UVs)" {from=MaterialExpressionDDX_2}
    "DDY(UVs)" {from=MaterialExpressionDDY_2}
    Apply View MipBias {from=""}
   ]
   properties {
    constMipValue=0
    mipValueMode=TMVM_Derivative
    samplerSource=SSM_Clamp_WorldGroupSettings
    samplerType=SAMPLERTYPE_Normal
    texture=&T_waterSplash_N
   }
  }
  MaterialExpressionTextureSample_18 {
   class=MaterialExpressionTextureSample
   inputs [
    UVs {from=MaterialExpressionAdd_15}
    Tex {from=""}
    Apply View MipBias {from=""}
   ]
   properties {samplerType=SAMPLERTYPE_Normal, texture=&T_Water2_N}
  }
  MaterialExpressionTextureSample_19 {
   class=MaterialExpressionTextureSample
   inputs [
    UVs {from=MaterialExpressionAdd_1}
    Tex {from=""}
    "DDX(UVs)" {from=MaterialExpressionDDX_0}
    "DDY(UVs)" {from=MaterialExpressionDDY_0}
    Apply View MipBias {from=""}
   ]
   properties {
    mipValueMode=TMVM_Derivative
    samplerSource=SSM_Clamp_WorldGroupSettings
    samplerType=SAMPLERTYPE_LinearColor
    texture=&T_waterSplash
   }
  }
  MaterialExpressionTextureSample_2 {
   class=MaterialExpressionTextureSample
   inputs [
    UVs {from=MaterialExpressionAdd_19}
    Tex {from=""}
    "DDX(UVs)" {from=MaterialExpressionDDX_9}
    "DDY(UVs)" {from=MaterialExpressionDDY_9}
    Apply View MipBias {from=""}
   ]
   properties {mipValueMode=TMVM_Derivative, texture=/Game/Effects/Fort_Effects/Textures/Water/T_BoilBubble_D}
  }
  MaterialExpressionTextureSample_25 {
   class=MaterialExpressionTextureSample
   inputs [
    UVs {from=MaterialExpressionAdd_34}
    Tex {from=""}
    Apply View MipBias {from=""}
   ]
   properties {samplerType=SAMPLERTYPE_Normal, texture=&T_Water2_N}
  }
  MaterialExpressionTextureSample_3 {
   class=MaterialExpressionTextureSample
   inputs [
    UVs {from=MaterialExpressionAdd_20}
    Tex {from=""}
    Apply View MipBias {from=""}
   ]
   properties {texture=/Game/Effects/Fort_Effects/Textures/Liquid/T_FX_Splash_01}
  }
  MaterialExpressionTextureSample_4 {
   class=MaterialExpressionTextureSample
   inputs [
    UVs {from=MaterialExpressionMultiply_12}
    Tex {from=""}
    "DDX(UVs)" {from=MaterialExpressionDDX_1}
    "DDY(UVs)" {from=MaterialExpressionDDY_1}
    Apply View MipBias {from=""}
   ]
   properties {
    mipValueMode=TMVM_Derivative
    samplerSource=SSM_Clamp_WorldGroupSettings
    samplerType=SAMPLERTYPE_Normal
    texture=&T_waterSplash_N
   }
  }
  MaterialExpressionTextureSample_5 {
   class=MaterialExpressionTextureSample
   inputs [
    UVs {from=MaterialExpressionMultiply_1}
    Tex {from=""}
    Apply View MipBias {from=""}
   ]
   properties {
    samplerSource=SSM_Clamp_WorldGroupSettings
    samplerType=SAMPLERTYPE_LinearColor
    texture=&T_waterSplash
   }
  }
  MaterialExpressionTextureSample_6 {
   class=MaterialExpressionTextureSample
   inputs [
    UVs {from=MaterialExpressionReroute_8}
    Tex {from=""}
    Apply View MipBias {from=""}
   ]
   properties {
    samplerSource=SSM_Clamp_WorldGroupSettings
    samplerType=SAMPLERTYPE_LinearColor
    texture=&T_waterSplash
   }
  }
  MaterialExpressionTextureSample_7 {
   class=MaterialExpressionTextureSample
   inputs [
    UVs {from=MaterialExpressionAdd_3}
    Tex {from=""}
    Apply View MipBias {from=""}
   ]
   properties {
    samplerSource=SSM_Clamp_WorldGroupSettings
    samplerType=SAMPLERTYPE_Normal
    texture=&T_waterSplash_N
   }
  }
  MaterialExpressionTextureSample_8 {
   class=MaterialExpressionTextureSample
   inputs [
    UVs {from=MaterialExpressionAdd_18}
    Tex {from=""}
    Level {from=""}
    Apply View MipBias {from=""}
   ]
   properties {constMipValue=0, mipValueMode=TMVM_MipLevel, texture=&T_BoilBubble_D}
  }
  MaterialExpressionTextureSample_9 {
   class=MaterialExpressionTextureSample
   inputs [
    UVs {from=MaterialExpressionAdd_21}
    Tex {from=""}
    "DDX(UVs)" {from=MaterialExpressionDDX_10}
    "DDY(UVs)" {from=MaterialExpressionDDY_10}
    Apply View MipBias {from=""}
   ]
   properties {constMipValue=0, mipValueMode=TMVM_Derivative, texture=&T_FX_Splash_01}
  }
  MaterialExpressionVectorParameter_0 {class=MaterialExpressionVectorParameter, properties {group=Color, parameterName=ColorA, defaultValue=<rgba 0 0.1979 0 1>}}
  MaterialExpressionVectorParameter_11 {class=MaterialExpressionVectorParameter, properties {group=Caustics, parameterName=Vat Velocity}}
  MaterialExpressionVectorParameter_12 {class=MaterialExpressionVectorParameter, properties {group=Caustics, parameterName=nynt, defaultValue=<rgba 0.5 0.5 0.25 0>}}
  MaterialExpressionVectorParameter_13 {class=MaterialExpressionVectorParameter, properties {group=Caustics, parameterName=Caustics, defaultValue=<rgba 0.0008 0.026 0 1>}}
  MaterialExpressionVectorParameter_4 {class=MaterialExpressionVectorParameter, properties {group=Color, parameterName=ColorB, defaultValue=<rgba 0.0016 0.1719 0 1>}}
  MaterialExpressionVectorParameter_6 {class=MaterialExpressionVectorParameter, properties {group=Normals, parameterName=NormalUVScale, defaultValue=<rgba 1 0.381 0 0>}}
  MaterialExpressionVectorParameter_7 {class=MaterialExpressionVectorParameter, properties {group=Normals, parameterName=Normal2UVScale, defaultValue=<rgba 1 0.4667 0 0>}}
  MaterialExpressionVectorParameter_8 {
   class=MaterialExpressionVectorParameter
   properties {group=Normals_Mobile, parameterName=NormalUVScale_M/M2, defaultValue=<rgba 1 0.381 0 0>}
  }
  MaterialExpressionVectorParameter_9 {
   class=MaterialExpressionVectorParameter
   properties {group=Normals_Mobile, parameterName=NormalUVScale_M3, sortPriority=1, defaultValue=<rgba 1 0.4667 0 0>}
  }
  MaterialExpressionVertexInterpolator_0 {class=MaterialExpressionVertexInterpolator, inputs [VS {from=MaterialExpressionMultiply_4, output=""}], properties {}}
  MaterialExpressionVertexInterpolator_1 {class=MaterialExpressionVertexInterpolator, inputs [VS {from=MaterialExpressionMultiply_4, output=""}], properties {}}
  MaterialExpressionVertexInterpolator_2 {class=MaterialExpressionVertexInterpolator, inputs [VS {from=MaterialExpressionMultiply_5, output=""}], properties {}}
  MaterialExpressionViewProperty_0 {class=MaterialExpressionViewProperty, inputs [], properties {property=MEVP_PostVolumeUserFlags}}
 }
 properties {
  bAutomaticallySetUsageInEditor=false
  blendMode=BLEND_Translucent
  materialDomain=MD_DeferredDecal
  pixelDepthOffsetMode=PDOM_Legacy
  shadingModel=MSM_Unlit
  usedShadingModels=Unlit
 }
}
